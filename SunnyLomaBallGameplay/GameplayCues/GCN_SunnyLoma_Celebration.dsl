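; Burst gameplay cue for a SunnyLoma ball celebration, layering radius scaling, a sound indicator and a physics push onto the parent effects.
; After the parent burst, a valid target has its first particle system scaled by the tuned radius relative to a default.
; PlaySoundIndicator shows a gunshot-type indicator using the first burst sound's max distance.
; A radial force field spawns only at higher material quality, outside PC performance mode and mobile.
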(fn OnBurst(EitherVFX-preferthisversion)
  (params (MyTarget :type Actor%20Object%20Reference) (Parameters :type Gameplay%20Cue%20Parameters%20Structure%20%28by%20ref%29) (ParticleComponents :type Array%20of%20FXSystem%20Component%20Object%20References) (AudioComponents :type Array%20of%20Audio%20Component%20Object%20References) (BurstCameraShakeInstance :type Camera%20Shake%20Base%20Object%20Reference) (BurstDecalInstance :type Decal%20Actor%20Object%20Reference))
  (bind _returnvalue (Transformation|GetActorLocation MyTarget))
  (|CallParentFunction MyTarget Parameters ParticleComponents AudioComponents BurstCameraShakeInstance BurstDecalInstance)
  (Utilities|IsValid MyTarget
    (:Is%20Valid
      (Effects|Components|ParticleSystem|SetFloatParameter (Utilities|Array|Get(acopy) ParticleComponents) "RadiusScalar" (/ (ScalableFloat|GetValueatLevel (Variables|Default|GetRow_Radius)) (LocalVariables|GetDefaultRadius)))
      (CallFunction|PlaySoundIndicator _returnvalue (Variables|/Script/Engine.Actor|GetInstigator MyTarget))
      (if (Math|Boolean|ANDBoolean (> (Rendering|Material|GetRenderingMaterialQualityLevel) 2) (not (Class|PlatformHelpersLibrary|IsPcPerfMode 0)) (not (TouchPlatform|IsMobileGame 0)))
        (Game|SpawnActorfromClass MyTarget "/Game/Effects/Niagara/Blueprints/PhsyicsFields/FN_RadialForce.FN_RadialForce_C" (Math|Transform|MakeTransform _returnvalue) "Undefined" "MultiplyWithRoot" 0 3.0 0.75 1e+03 1.5 0.05 0))
      (return))))

(fn PlaySoundIndicator
  (params (Location :type Vector) (Instigator :type Actor%20Object%20Reference))
  (bind _bursteffectdata (Variables|BurstEffects|GetBurstEffectData))
  (bind (_burstparticles _burstsounds) (Utilities|Struct|BreakFortBurstEffectData _bursteffectdata))
  (bind (_condition _attachment _soundcue) (Utilities|Struct|BreakFortGameplayCueAudioInfo (Utilities|Array|Get(aref) _burstsounds)))
  (if (> (Utilities|Array|Length _burstsounds) 0)
    (Fortnite|ShowSoundIndicatorLocal :Location Location :MaxAudibleDistance (Variables|/Script/Engine.SoundBase|GetMaxDistance _soundcue) :IndicatorType "Gunshot" :Instigator Instigator :Tint "(R=0.000000,G=0.000000,B=0.000000,A=0.000000)" :SoundBase _soundcue :AttenuationMaxRangeMultiplier 1.0 :AttenuationIntensityMultiplier 1.0) ; Indicator for fuse
  ))
